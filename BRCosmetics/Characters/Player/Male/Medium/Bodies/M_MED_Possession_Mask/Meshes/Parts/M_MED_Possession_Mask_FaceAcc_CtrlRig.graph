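# Control Rig for the face accessory of the male medium Possession Mask cosmetic, posing its dyn_helmet bones as the head and hair move.
# Parent constraints make dyn_helmet_01 follow dyn_Hair_BackR_03 in filtered rotation and scale, and dyn_helmet_03 follow dyn_helmet_01, both without translation.
# Three spherical pose readers on the head then weight corrective offsets on those helmet bones.
# A final dyn_helmet_03 modify node is unconnected to execution.
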ControlRigBlueprint /BRCosmetics/Characters/Player/Male/Medium/Bodies/M_MED_Possession_Mask/Meshes/Parts/M_MED_Possession_Mask_FaceAcc_CtrlRig
graphs [
 RigVMFunctionLibrary {links [], nodes []}
 RigVMModel {nodes [
   RigUnit_BeginExecution = &RigUnit_BeginExecution()
   &RigUnit_ParentConstraint(
    AdvancedSettings=(RotationOrderForFilter=XZY)
    Child=(Type=Bone,Name="dyn_helmet_01")
    Filter=(TranslationFilter=(bX=false,bY=false,bZ=false),RotationFilter=(bX=false,bY=True,bZ=True),ScaleFilter=(bX=True,bY=True,bZ=True))
    Parents=((Item=(Type=Bone,Name="dyn_Hair_BackR_03"),Weight=1))
    Weight=1
    ExecutePin=@RigUnit_BeginExecution.ExecutePin
   )
   > &RigUnit_ParentConstraint(
    AdvancedSettings=(RotationOrderForFilter=XZY)
    Child=(Type=Bone,Name="dyn_helmet_03")
    Filter=(TranslationFilter=(bX=false,bY=false,bZ=false),RotationFilter=(bX=true,bY=true,bZ=true),ScaleFilter=(bX=True,bY=True,bZ=True))
    Parents=((Item=(Type=Bone,Name="dyn_helmet_01"),Weight=1))
    Weight=1
   )
   > n1 = &RigUnit_SphericalPoseReader(
    ActiveRegionScaleFactors=(PositiveWidth=1,NegativeWidth=1,PositiveHeight=1,NegativeHeight=0.5)
    ActiveRegionSize=0.1
    Debug=(bDrawDebug=false,DebugScale=25,DebugSegments=20,DebugThickness=0.25)
    DriverAxis=(X=0,Y=1,Z=0)
    DriverItem=(Type=Bone,Name="head")
    FalloffRegionScaleFactors=(PositiveWidth=1,NegativeWidth=1,PositiveHeight=1.1,NegativeHeight=1)
    FalloffSize=0.3
    OptionalParentItem=(Type=Bone)
    RotationOffset=(X=130,Y=0,Z=0)
   )
   > &RigUnit_ModifyTransforms(
    ItemToModify="((Item=(Type=Bone,Name=\"dyn_helmet_01\"),Transform=(Rotation=(X=0,Y=0,Z=0,W=1),Translation=(X=0,Y=0,Z=0),Scale3D=(X=0.9,Y=0.9,Z=0.9))),(Item=(Type=Bone,Name=\"dyn_helmet_03\"),Transform=(Rotation=(X=0,Y=0,Z=0,W=1),Translation=(X=0,Y=0,Z=0),Scale3D=(X=1,Y=1,Z= ...(truncated)"
    ItemToModify.0.Transform.Rotation=&RigVMFunction_MathQuaternionFromEuler(Euler=(X=-15,Y=0,Z=-20)).Result
    ItemToModify.1.Transform.Rotation=&RigVMFunction_MathQuaternionFromEuler(Euler=(X=0,Y=0,Z=-20)).Result
    Weight=@n1.OutputParam
   )
   > n2 = &RigUnit_SphericalPoseReader(
    ActiveRegionScaleFactors=(PositiveWidth=1,NegativeWidth=0.5,PositiveHeight=1.5,NegativeHeight=1.5)
    ActiveRegionSize=0.1
    Debug=(bDrawDebug=false,DebugScale=25,DebugSegments=20,DebugThickness=0.25)
    DriverAxis=(X=0,Y=1,Z=0)
    DriverItem=(Type=Bone,Name="head")
    FalloffRegionScaleFactors=(PositiveWidth=1,NegativeWidth=0.5,PositiveHeight=1.5,NegativeHeight=1.5)
    FalloffSize=0.2
    FlipWidthScaling=false
    OptionalParentItem=(Type=Bone)
    RotationOffset=(X=90,Y=0,Z=-25)
   )
   > &RigUnit_ModifyTransforms(
    ItemToModify=((Item=(Type=Bone,Name="dyn_helmet_01"),Transform=(Rotation=(X=0,Y=0,Z=0,W=1),Translation=(X=0,Y=0,Z=0),Scale3D=(X=1,Y=1,Z=1))))
    ItemToModify.0.Transform.Rotation=&RigVMFunction_MathQuaternionFromEuler(Euler=(X=0,Y=0,Z=-20)).Result
    Weight=@n2.OutputParam
   )
   > n3 = &RigUnit_SphericalPoseReader(
    ActiveRegionScaleFactors=(PositiveWidth=2,NegativeWidth=0.5,PositiveHeight=2,NegativeHeight=2)
    ActiveRegionSize=0.075
    Debug=(bDrawDebug=false,DebugScale=25,DebugSegments=20,DebugThickness=0.25)
    DriverAxis=(X=0,Y=1,Z=0)
    DriverItem=(Type=Bone,Name="head")
    FalloffRegionScaleFactors=(PositiveWidth=1,NegativeWidth=0.4,PositiveHeight=1,NegativeHeight=1)
    FalloffSize=0.2
    FlipHeightScaling=false
    FlipWidthScaling=true
    OptionalParentItem=(Type=Bone)
    RotationOffset=(X=90,Y=0,Z=12.5)
   )
   > &RigUnit_ModifyTransforms(
    ItemToModify=((Item=(Type=Bone,Name="dyn_helmet_01"),Transform=(Rotation=(X=0,Y=0,Z=0,W=1),Translation=(X=0,Y=0,Z=0),Scale3D=(X=1,Y=1,Z=1))))
    ItemToModify.0.Transform.Rotation=&RigVMFunction_MathQuaternionFromEuler(Euler=(X=0,Y=0,Z=-10)).Result
    Weight=@n3.OutputParam
   )
   &RigUnit_ModifyTransforms(
    ExecutePin=()
    ItemToModify=((Item=(Type=Bone,Name="dyn_helmet_03"),Transform=(Rotation=(X=0,Y=0,Z=0,W=1),Translation=(X=0,Y=0,Z=0),Scale3D=(X=1,Y=1,Z=1))))
    Weight=1
    ItemToModify.0.Transform.Rotation=&RigVMFunction_MathQuaternionFromEuler(Euler=(X=0,Y=0,Z=-5)).Result
   )
  ]}
]
hierarchy [
 root {loc=<0 0 0>, rot=<0 0 0 1>}
 pelvis {parent=root, loc=<0 -0.7071 93.1191>, rot=<0.0127 0.707 -0.0127 -0.707>}
 spine_01 {parent=pelvis, loc=<6.7494 -0.6388 0>, rot=<0 0 0.034 -0.9994>}
 spine_02 {parent=spine_01, loc=<7.2518 -0.035 0>, rot=<0 0 0.0238 -0.9997>}
 spine_03 {parent=spine_02, loc=<6.0104 0 0>, rot=<0 0 0.0149 -0.9999>}
 spine_04 {parent=spine_03, loc=<8.1024 0 0>, rot=<0 0 -0.0132 -0.9999>}
 spine_05 {parent=spine_04, loc=<10.0146 0 0>, rot=<0 0 -0.0015 -1>}
 neck_01 {parent=spine_05, loc=<11.3138 0.1529 0>, rot=<0 0 -0.2003 -0.9797>}
 neck_02 {parent=neck_01, loc=<4.5627 0 0>, rot=<0 0 0.0504 -0.9987>}
 head {parent=neck_02, loc=<5.3914 0 0>, rot=<0 0 0.115 -0.9934>}
 dyn_Hair_BackR_01 {parent=head, loc=<8.1854 -0.439 4.9819>, rot=<0 0 0.0928 -0.9957>}
 dyn_Hair_BackR_02 {parent=dyn_Hair_BackR_01, loc=<-1.9753 -3.7984 0>, rot=<0 0 0.5593 -0.8289>}
 dyn_Hair_BackR_03 {parent=dyn_Hair_BackR_02, loc=<0.1785 -6.9081 0.4801>, rot=<0 0 -0.004 -1>}
 dyn_Hair_BackR_04 {parent=dyn_Hair_BackR_03, loc=<2.7527 -8.9279 0.7127>, rot=<0 0 0 1>}
 dyn_Hair_BackR_05 {parent=dyn_Hair_BackR_04, loc=<1.1629 -7.0493 0.5082>, rot=<0 0 0 1>}
 dyn_Hair_BackR_06 {parent=dyn_Hair_BackR_05, loc=<-0.97 -7.2564 0.1543>, rot=<0 0 0.0617 -0.9981>}
 dyn_Hair_BackR_07 {parent=dyn_Hair_BackR_06, loc=<-1.6115 -5.5548 -0.067>, rot=<0 0 0 1>}
 dyn_Hair_BackR_08 {parent=dyn_Hair_BackR_07, loc=<-2.0374 -3.7958 -0.2177>, rot=<0 0 0 1>}
 dyn_Hair_Back_01 {parent=head, loc=<8.5016 -1.2653 0.5844>, rot=<0 0 0.0928 -0.9957>}
 dyn_Hair_Back_02 {parent=dyn_Hair_Back_01, loc=<-1.9753 -3.7984 0>, rot=<0 0 0.5593 -0.8289>}
 dyn_Hair_Back_03 {parent=dyn_Hair_Back_02, loc=<0.0293 -6.9926 0>, rot=<0 0 -0.004 -1>}
 dyn_Hair_Back_04 {parent=dyn_Hair_Back_03, loc=<2.2287 -9.0584 0.0849>, rot=<0 0 0 1>}
 dyn_Hair_Back_05 {parent=dyn_Hair_Back_04, loc=<1.1349 -8.8919 -0.5439>, rot=<0 0 0 1>}
 dyn_Hair_Back_06 {parent=dyn_Hair_Back_05, loc=<-1.7604 -7.4099 -0.2152>, rot=<0 0 0.0617 -0.9981>}
 dyn_Hair_Back_07 {parent=dyn_Hair_Back_06, loc=<-1.3114 -5.2345 -0.2173>, rot=<0 0 0 1>}
 dyn_Hair_Back_08 {parent=dyn_Hair_Back_07, loc=<-2.0094 -4.9407 -0.2991>, rot=<0 0 0 1>}
 dyn_Hair_BackL_01 {parent=head, loc=<7.9722 -1.0353 -4.0889>, rot=<0 0 0.0928 -0.9957>}
 dyn_Hair_BackL_02 {parent=dyn_Hair_BackL_01, loc=<-1.9753 -3.7984 0>, rot=<0 0 0.5593 -0.8289>}
 dyn_Hair_BackL_03 {parent=dyn_Hair_BackL_02, loc=<-0.2757 -6.9116 0>, rot=<0 0 -0.004 -1>}
 dyn_Hair_BackL_04 {parent=dyn_Hair_BackL_03, loc=<2.9245 -9.0113 -0.1535>, rot=<0 0 0 1>}
 dyn_Hair_BackL_05 {parent=dyn_Hair_BackL_04, loc=<1.2094 -7.0493 -0.769>, rot=<0 0 0 1>}
 dyn_Hair_BackL_06 {parent=dyn_Hair_BackL_05, loc=<-2.1353 -7.4711 -1.887>, rot=<0 0 0.0617 -0.9981>}
 dyn_Hair_BackL_07 {parent=dyn_Hair_BackL_06, loc=<-1.7408 -4.8709 -2.3007>, rot=<0 0 0 1>}
 dyn_hair_l_fr_a_MSTR {parent=head, loc=<0.4806 3.6857 -6.4834>, rot=<0.0041 -1 0 0>}
 dyn_hair_l_fr_a_01 {parent=dyn_hair_l_fr_a_MSTR, loc=<0.0831 0 0>, rot=<0 0 0 1>}
 dyn_hair_l_Front_MSTR {parent=head, loc=<0.9561 1.6481 -6.6576>, rot=<0.0919 -0.9926 0.0067 -0.0796>}
 dyn_hair_l_Front_00 {parent=dyn_hair_l_Front_MSTR, loc=<0.5 0 0>, rot=<0.0038 -0.0632 -0.1166 -0.9912>}
 dyn_hair_l_Front_01 {parent=dyn_hair_l_Front_00, loc=<6.4093 0 0>, rot=<-0.0194 -0.0914 -0.1504 -0.9842>}
 dyn_hair_l_Front_02 {parent=dyn_hair_l_Front_01, loc=<6.0676 0 0>, rot=<-0.1101 -0.2165 -0.1058 -0.9643>}
 dyn_hair_r_Fr_MSTR {parent=head, loc=<0.1947 2.2616 5.1475>, rot=<0.1035 -0.9868 -0.012 0.124>}
 dyn_hair_r_Fr_00 {parent=dyn_hair_r_Fr_MSTR, loc=<0.4731 0 0>, rot=<-0.0067 0.0506 -0.0683 -0.9964>}
 dyn_hair_r_Fr_01 {parent=dyn_hair_r_Fr_00, loc=<3.644 0 0>, rot=<0.0012 0.0131 -0.1552 -0.9878>}
 dyn_hair_r_F_02 {parent=dyn_hair_r_Fr_01, loc=<4.892 0 0>, rot=<0.0309 0.0697 -0.1677 -0.9829>}
 dyn_hair_r_Fr_03 {parent=dyn_hair_r_F_02, loc=<4.4193 0 0>, rot=<0.0189 0.0306 0.0317 -0.9989>}
 dyn_hair_r_Fr_04 {parent=dyn_hair_r_Fr_03, loc=<5.1383 0 0>, rot=<-0.0011 -0.0026 0.1204 -0.9927>}
 dyn_hair_r_Fr_01_a {parent=dyn_hair_r_Fr_00, loc=<3.7619 2.2635 1.7588>, rot=<0.9202 0.0576 0.0559 -0.3831>}
 dyn_helmet_MSTR {parent=head, loc=<6.5666 -1.3708 0>, rot=<0.5254 -0.8509 0 0>}
 dyn_helmet_01 {parent=dyn_helmet_MSTR, loc=<4.712 0 0>, rot=<0 0 -0.4769 -0.8789>}
 dyn_helmet_02 {parent=dyn_helmet_01, loc=<5.1935 0 0>, rot=<0 0 0 1>}
 dyn_helmet_03 {parent=dyn_helmet_MSTR, loc=<3.5407 2.0238 -5.0655>, rot=<0 0 -0.4769 -0.8789>}
 clavicle_l {parent=spine_05, loc=<4.3575 -0.1466 -2.3223>, rot=<0.0465 -0.7289 -0.009 -0.6829>}
 upperarm_l {parent=clavicle_l, loc=<15.5111 0 0>, rot=<0.0226 -0.3234 -0.0088 -0.9459>}
 clavicle_r {parent=spine_05, loc=<4.3575 -0.1466 2.3223>, rot=<0.7289 0.0465 -0.6829 0.009>}
 upperarm_r {parent=clavicle_r, loc=<-15.5111 0 0>, rot=<0.0226 -0.3234 -0.0088 -0.9459>}
 upperarm_twist_01_r {parent=upperarm_r, loc=<-0.0041 0 0>, rot=<0 0 0 -1>}
]
